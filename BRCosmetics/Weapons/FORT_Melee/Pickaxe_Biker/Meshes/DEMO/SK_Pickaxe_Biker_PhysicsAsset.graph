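# Physics asset for the Biker pickaxe skeletal mesh, a two-handed melee weapon cosmetic.
# A single capsule body on the Melee_TwoHanded_Handle bone is offset and tilted from the bone origin.
# It has no constraints, so there are no jointed parts and the collision is one simple shape.
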PhysicsAsset /BRCosmetics/Weapons/FORT_Melee/Pickaxe_Biker/Meshes/DEMO/SK_Pickaxe_Biker_PhysicsAsset
bodies {Melee_TwoHanded_Handle [{shapeType=Capsule, length=78.8961, radius=43.4601, center=<-0.3791 3.1014 47.0856>, rotation=<rot -4.9494 -90 7.0554>}]}
constraints []
